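# Masked, two-sided lit material for AvacadoEater splash effects on Niagara sprite and mesh particles, using water splash textures.
# A center-scaled radial flowmap, weighted by the CenterWarp dynamic parameter, distorts tiled UVs for the normal and mask textures.
# Particle alpha drives a power-based erosion of the selected mask channel, multiplied by alpha and an opacity channel.
# Particle color feeds optional tinting and scaled emissive, and the normal can be flattened.
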Material /MotherGameplay/Items/AvacadoEater/Assets/FX/MI_AvacadoEater_Splash
graph {
 properties {
  bUsedWithNiagaraMeshParticles=true
  bUsedWithNiagaraSprites=true
  blendMode=BLEND_Masked
  pixelDepthOffsetMode=PDOM_Legacy
  twoSided=true
  usedShadingModels=Default Lit
 }
 nodes [
  Constant() {r=0.2}
  n1 = ParticleColor()
  LinearInterpolate(B=@n1.RGB, Alpha=ScalarParameter() {parameterName=UseParticleColor})
  n2 = VectorParameter() {parameterName=Tile, defaultValue=<rgba 2 1 0 0>}
  n3 = MaterialFunctionCall(Texture Scale=Constant() {r=1.4}) {materialFunction=&ScaleUVsByCenter}
  n4 = Subtract(A=ComponentMask(None=TextureSample(UVs=@n3.UVs) {samplerType=SAMPLERTYPE_LinearColor, texture=/Game/Environments/Hexmap/HexClouds/T_RadialFlowmap01}.RGB) {g=true, r=true}) {constB=0.5}
  n5 = DynamicParameter() {parameterIndex=1, defaultValue=<rgba 0 1 1 1>, paramNames [CenterWarp,Param2,Param3,Param4]}
  n6 = Multiply(A=AppendVector(A=@n2.R, B=@n2.R), B=Add(A=TextureCoordinate(), B=Multiply(A=@n4, B=@n5.CenterWarp)))
  n7 = TextureSampleParameter2D(UVs=@n6) {
   parameterName=Normal
   samplerType=SAMPLERTYPE_Normal
   texture=/Game/Packages/Fortress_Transylvania/Haunted_Forest/Swamps/Textures/T_water_splash_05_n
  }
  LinearInterpolate(A=@n7.RGB, B=Constant3Vector() {constant=<rgba 0 0 1 0>}, Alpha=ScalarParameter() {parameterName=Flatten Normal})
  MaterialFunctionCall(Emissive=Multiply(A=@n1.RGB, B=ScalarParameter() {defaultValue=0.15, parameterName=EmissiveAdd}) {constB=0.15}) {materialFunction=&MF_FX_EmissiveAdjustments}
  n8 = TextureSampleParameter2D(UVs=@n6) {parameterName=Mask, texture=/Game/Packages/Fortress_Transylvania/Haunted_Forest/Swamps/Textures/T_water_splash_05}
  n9 = ChannelMaskParameter(None=@n8.RGB) {maskChannel=Blue, parameterName=ErodeMask, defaultValue=<rgba 0 0 1 0>}
  n10 = LinearInterpolate(Alpha=@n1.A) {constA=2, constB=-0.1}
  Multiply(A=Multiply(A=@n1.A, B=Power(Base=@n9.R, Exp=@n10)), B=ChannelMaskParameter(None=@n8.A) {parameterName=OpacityMask}.R)
 ]
}
