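# Translucent Niagara mesh-particle material that projects a tinted Bluglo splat as a fake decal for the AvacadoEater effects.
# Scene position behind the translucent surface, converted to particle space, gives planar XY UVs, with depth testing disabled.
# Splat green blends two colors, tinted by particle color for emissive; red, particle alpha and a sine fade along local Z form an opacity mask.
# A flattened splat normal map adds surface detail.
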Material /MotherGameplay/Items/AvacadoEater/Assets/FX/M_AvacadoEater_Decal
graph {
 properties {
  bDisableDepthTest=true
  bUsedWithNiagaraMeshParticles=true
  blendMode=BLEND_Translucent
  pixelDepthOffsetMode=PDOM_Legacy
  translucencyLightingMode=TLM_VolumetricPerVertexDirectional
  twoSided=true
  usedShadingModels=Default Lit
 }
 nodes [
  MaterialFunctionCall(
   Normal=TextureSample() {samplerType=SAMPLERTYPE_Normal, texture=/Game/Missions/Secondary/BluGloSiphon/Textures/T_Bluglo_Splat_01_N}.RGB
   Flatness=ScalarParameter() {parameterName="Flatness (S)"}
  ) {materialFunction=&FlattenNormal}
  n1 = Subtract(
   A=MaterialFunctionCall() {materialFunction=&WorldPositionBehindTranslucency}.Result
   B=ParticlePositionWS()
  )
  n2 = Add(
   A=Divide(
    A=Transform(None=@n1) {transformSourceType=TRANSFORMSOURCE_World, transformType=TRANSFORM_ParticleWorld}
    B=ScalarParameter() {defaultValue=200}
   )
   B=ScalarParameter() {defaultValue=0.5, parameterName=Param_1}
  )
  n3 = TextureSample(UVs=Saturate(None=ComponentMask(None=@n2) {g=true, r=true})) {texture=/Game/Missions/Secondary/BluGloSiphon/Textures/T_Bluglo_Splat_01}
  n4 = LinearInterpolate(
   A=VectorParameter() {parameterName=Param_1, defaultValue=<rgba 0.0941 0.0081 0.2 1>}.RGB
   B=VectorParameter() {defaultValue=<rgba 0.0969 0.0969 0.9688 1>}.RGB
   Alpha=@n3.G
  )
  n5 = ParticleColor()
  MaterialFunctionCall(Emissive=Multiply(A=Multiply(A=@n4, B=@n5.RGB), B=ScalarParameter() {defaultValue=0.2, parameterName=Emissive_Mul})) {materialFunction=&MF_FX_EmissiveAdjustments}
  n6 = Multiply(A=Saturate(None=ComponentMask(None=@n2) {b=true})) {constB=0.5}
  Multiply(A=Multiply(A=@n5.A, B=@n3.R), B=Saturate(None=Sine(None=@n6)))
 ]
}
