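# MF_Avocado_Vines is a material function that computes a wobble offset for the AvacadoEater avocado's vines.
# A sine of time times speed, shifted by an offset scaled with the Egg Vines Frequency parameter, drives the motion.
# That offset previews as horizontal distance from the object pivot, rippling the wave outward or inward.
# Vertex color remapped to signed values gives direction, scaled by an amplitude that a mask input can invert.
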MaterialFunction /MotherGameplay/Items/AvacadoEater/Assets/Materials/MF_Avocado_Vines
graph {nodes [
  n1 = Time()
  n2 = FunctionInput() {bUsePreviewValueAsDefault=true, inputName=Speed Multiplier, inputType=FunctionInput_Scalar, previewValue=<1 0 0 1>}
  n3 = MaterialFunctionCall() {materialFunction=&ObjectPivotPoint}
  n4 = WorldPosition()
  n5 = FunctionInput(Preview=Distance(A=ComponentMask(None=@n3.Object Pivot Location) {g=true, r=true}, B=ComponentMask(None=@n4.XYZ) {g=true, r=true})) {bUsePreviewValueAsDefault=true, inputName=Wobble Offset, inputType=FunctionInput_Scalar}
  n6 = Divide(B=ScalarParameter() {defaultValue=2, parameterName=Egg Vines Frequency}) {constA=100}
  n7 = Sine(None=Add(A=Multiply(A=@n1, B=@n2), B=Divide(A=@n5, B=@n6) {constB=80}))
  n8 = Subtract(A=Multiply(A=VertexColor()) {constB=2})
  n9 = FunctionInput() {bUsePreviewValueAsDefault=true, inputName=Amplitude, inputType=FunctionInput_Scalar, previewValue=<0.2 0 0 1>}
  n10 = Multiply(A=@n9) {constB=-1}
  n11 = FunctionInput() {bUsePreviewValueAsDefault=true, inputName=Invert Direction Mask, inputType=FunctionInput_Scalar}
  FunctionOutput(None=Multiply(A=Multiply(A=@n7, B=@n8), B=LinearInterpolate(A=@n9, B=@n10, Alpha=@n11)))
 ]}
