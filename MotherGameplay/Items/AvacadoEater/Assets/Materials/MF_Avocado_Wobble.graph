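# Material function for the avocado wobble, producing a vertex offset vector relative to an influencer position.
# Vertices inside the radius are pushed out to the sphere's surface along the direction from the influencer.
# A second term, weighted by a local-height gradient, pulls vertices outside the radius toward that surface.
# A separate "New PP" branch computes an XY-shifted position from influencer, gradient and strength, but does not reach the output.
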MaterialFunction /MotherGameplay/Items/AvacadoEater/Assets/Materials/MF_Avocado_Wobble
graph {nodes [
  n1 = WorldPosition()
  n2 = ComponentMask(None=TransformPosition(None=@n1.XYZ) {transformSourceType=TRANSFORMPOSSOURCE_World}) {b=true}
  n3 = FunctionInput(Preview=Divide(A=Subtract(A=@n2) {constB=10}) {constB=50}) {bUsePreviewValueAsDefault=true, inputName=Gradient, inputType=FunctionInput_Scalar, previewValue=<1 0 0 1>}
  n4 = FunctionInput(Preview=TransformPosition(None=Constant3Vector() {constant=<rgba 0 0 37.5 0>}) {transformType=TRANSFORMPOSSOURCE_World}) {bUsePreviewValueAsDefault=true, inputName=Influencer Position}
  n5 = Multiply(A=@n4, B=Saturate(None=@n3))
  n6 = Constant3Vector() {constant=<rgba 1 1 0 0>}
  n7 = FunctionInput() {bUsePreviewValueAsDefault=true, inputName=Str, inputType=FunctionInput_Scalar, previewValue=<0.6 0 0 1>}
  Add(A=Multiply(A=Multiply(A=@n5, B=@n6), B=@n7), B=@n1.XYZ) {desc=New PP}
  n8 = FunctionInput(Preview=@n1.XYZ) {bUsePreviewValueAsDefault=true, inputName=Vertex Position}
  n9 = Subtract(A=@n8, B=@n4)
  n10 = FunctionInput() {bUsePreviewValueAsDefault=true, inputName=Radius, inputType=FunctionInput_Scalar, previewValue=<22.5 0 0 1>}
  n11 = Subtract(A=@n10, B=Distance(A=@n8, B=@n4))
  n12 = Clamp(None=@n11) {clampMode=CMODE_ClampMin}
  n13 = ComponentMask(None=TransformPosition(None=WorldPosition().XYZ) {transformSourceType=TRANSFORMPOSSOURCE_World}) {b=true}
  n14 = FunctionInput(Preview=Saturate(None=Divide(A=@n13) {constB=100})) {bUsePreviewValueAsDefault=true, inputName=Gradient2, inputType=FunctionInput_Scalar, previewValue=<1 0 0 1>}
  n15 = Multiply(A=Clamp(None=@n11) {clampMode=CMODE_ClampMax}, B=@n14)
  FunctionOutput(None=Multiply(A=Normalize(VectorInput=@n9), B=Add(A=@n12, B=@n15)))
 ]}
