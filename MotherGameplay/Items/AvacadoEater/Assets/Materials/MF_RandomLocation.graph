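# MF_RandomLocation is a material function in the AvacadoEater item's assets that outputs a continuously moving offset vector, deterministic despite the name.
# Time scaled by the Speed input drives cosine and sine waves for X, Y and Z, with Y using a different period and a divide to vary the path.
# The vector is normalized and scaled by Max Distance, so the offset travels on a sphere of that radius.
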MaterialFunction /MotherGameplay/Items/AvacadoEater/Assets/Materials/MF_RandomLocation
graph {nodes [
  n1 = FunctionInput() {bUsePreviewValueAsDefault=true, inputName=Speed, inputType=FunctionInput_Scalar, previewValue=<0.2 0 0 1>}
  n2 = Multiply(A=Time(), B=@n1)
  n3 = MaterialFunctionCall(X=Cosine(None=@n2), Y=Divide(A=Sine(None=@n2) {period=0.7}), Z=Sine(None=@n2)) {materialFunction=&MakeFloat3}
  n4 = FunctionInput() {bUsePreviewValueAsDefault=true, inputName=Max Distance, inputType=FunctionInput_Scalar, previewValue=<1 0 0 1>}
  FunctionOutput(None=Multiply(A=Normalize(VectorInput=@n3.Result), B=@n4))
 ]}
