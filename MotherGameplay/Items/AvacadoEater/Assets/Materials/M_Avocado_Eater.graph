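# Masked material for the Avocado Eater gadget, adding animated vines, wobble, a chomping motion and a timed dissolve.
# Texture-masked vine and wobble functions build the vertex offset, with the wobble influencer moved by random motion and a sine chomp.
# Trigger and player-direction reroutes are declared for the animation.
# Custom primitive data drives the trigger and dissolve, which uses growing sphere masks for opacity and green emissive.
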Material /MotherGameplay/Items/AvacadoEater/Assets/Materials/M_Avocado_Eater
graph {
 properties {bTangentSpaceNormal=false, blendMode=BLEND_Masked, pixelDepthOffsetMode=PDOM_Legacy, usedShadingModels=Default Lit}
 nodes [
  n1 = TextureSample() {samplerType=SAMPLERTYPE_Masks, texture=/MotherGameplay/Items/AvacadoEater/Assets/Textures/T_Avocado_Eater_Gadget_FX}
  n2 = Power(Base=TextureSample() {samplerType=SAMPLERTYPE_Masks, texture=/MotherGameplay/Items/AvacadoEater/Assets/Textures/T_Avocado_Eater_Gadget_MasksWPO}.G) {constExponent=3}
  n3 = Constant3Vector()
  n4 = MaterialFunctionCall!(
   Invert Direction Mask
   Amplitude=Multiply(A=@n1.R, B=Multiply(A=@n2) {constB=0.5})
   Speed Multiplier=Constant() {r=0.5}
   Wobble Offset=Distance(A=TransformPosition(None=@n3) {transformType=TRANSFORMPOSSOURCE_World}, B=WorldPosition().XYZ)
  ) {materialFunction=&/MF_Avocado_Vines}
  n5 = MaterialFunctionCall() {materialFunction=&ObjectPivotPoint}
  n6 = NamedRerouteUsage()
  n7 = LinearInterpolate(Alpha=@n6) {constA=0.7, constB=0.1}
  n8 = Constant() {r=2}
  n9 = Constant3Vector() {constant=<rgba 0 0 37.5 0>}
  n10 = ObjectOrientation()
  n11 = Multiply(A=Time(), B=LinearInterpolate(Alpha=@n6) {constA=3, constB=0}) {constB=3}
  n12 = Multiply(A=ConstantBiasScale(None=Sine(None=@n11)), B=NamedRerouteUsage())
  n13 = ScalarParameter() {defaultValue=15, parameterName=Chomp Distance}
  n14 = Add(
   A=Add(
    A=MaterialFunctionCall!(Speed=@n7, Max Distance=@n8) {materialFunction=&/MF_RandomLocation}.Result
    B=MaterialFunctionCall!(VectorToRotate=@n9, LookAtVector=@n10, RestingVector) {materialFunction=/Engine/Functions/Engine_MaterialFunctions01/Vectors/RotateVector}.Result
   )
   B=Multiply(A=Multiply(A=@n12, B=@n13) {constB=10}, B=LinearInterpolate(Alpha=@n6) {constA=1, constB=0})
  )
  Add(
   A=Multiply(A=@n4.Result)
   B=MaterialFunctionCall!(Vertex Position, Str, Radius, Gradient, Gradient2, Influencer Position=Add(A=@n5.Object Pivot Location, B=@n14)) {materialFunction=&/MF_Avocado_Wobble}.Result
  ) {constB=0}
  Constant() {r=0.2}
  LinearInterpolate(
   A=Transform(None=TextureSample() {samplerType=SAMPLERTYPE_Normal, texture=/MotherGameplay/Items/AvacadoEater/Assets/Textures/T_Avocado_Eater_Gadget_N}.RGB)
   B=Add(A=VertexNormalWS(), B=@n4.Result)
  )
  n15 = VertexNormalWS()
  n16 = ConstantBiasScale(None=Multiply(A=ComponentMask(None=@n15) {b=true}) {constB=-1})
  Multiply(A=Multiply(A=Constant3Vector() {constant=<rgba 0.0123 0.7011 0.0742 1>}, B=Power(Base=@n16) {constExponent=5}), B=@n1.B)
  n17 = TextureSample() {samplerType=SAMPLERTYPE_Masks, texture=/MotherGameplay/Items/AvacadoEater/Assets/Textures/T_Avocado_Eater_Gadget_S}
  Multiply(A=@n17.B, B=Constant() {r=0.5})
  n18 = WorldPosition()
  n19 = Subtract(
   A=Subtract(A=Time(), B=ScalarParameter() {defaultValue=5, parameterName=Delay})
   B=ScalarParameter() {bUseCustomPrimitiveData=true, parameterName=DissolveTime, primitiveDataIndex=2}
  )
  n20 = ScalarParameter() {defaultValue=1, parameterName=DissolveAnimDuration}
  n21 = Add(A=Saturate(None=Divide(A=@n19, B=@n20)), B=Divide(A=NamedRerouteUsage()) {constB=4})
  n22 = Add(
   A=ObjectPositionWS()
   B=LinearInterpolate(A=Constant3Vector() {constant=<rgba 0 0 12 0>}, B=Constant3Vector(), Alpha=Power(Base=@n21) {constExponent=3})
  )
  n23 = OneMinus(None=SphereMask(A=@n18.XYZ, B=@n22, Radius=LinearInterpolate(Alpha=@n21) {constB=80}, Hardness=Constant()))
  n24 = ScalarParameter() {bUseCustomPrimitiveData=true, defaultValue=1, parameterName=Dissolve, primitiveDataIndex=1}
  NamedRerouteDeclaration(None=LinearInterpolate(B=Power(Base=@n23) {constExponent=3}, Alpha=@n24) {constA=1}) {name=DissolveAnimationOpacity, nodeColor=<rgba 1 0 0.3281 1>}
  NamedRerouteDeclaration(None=@n17.B) {name=RoughnessNoise, nodeColor=<rgba 0 0.5312 1 1>}
  n25 = SphereMask(A=@n18.XYZ, B=@n22, Radius=LinearInterpolate(Alpha=@n21) {constB=61})
  n26 = Constant3Vector() {constant=<rgba 0.0138 0.8 0.0629 1>}
  NamedRerouteDeclaration(None=LinearInterpolate(B=Multiply(A=@n25, B=@n26), Alpha=@n24)) {name=DissolveAnimationEmissive, nodeColor=<rgba 1 0 0.5391 1>}
  NamedRerouteDeclaration(None=OneMinus(None=ScalarParameter() {bUseCustomPrimitiveData=true, parameterName=Trigger})) {name=Trigger, nodeColor=<rgba 0 0.25 1 1>}
  n27 = CollectionParameter() {collection=&FortniteMaterialParameters, parameterName=PlayerPosition}
  n28 = ComponentMask(None=@n27) {b=true, g=true, r=true}
  n29 = MaterialFunctionCall() {materialFunction=&ObjectPivotPoint}
  NamedRerouteDeclaration(None=MaterialFunctionCall(Vector=Subtract(A=@n28, B=@n29.Object Pivot Location), Default=Constant3Vector() {constant=<rgba 1 0 0 0>}) {materialFunction=&SafeNormalize}.Result) {name=Player Direction, nodeColor=<rgba 1 0.1172 0 1>}
  NamedRerouteUsage()
  NamedRerouteUsage()
  n30 = Constant() {r=1000}
  n31 = CollectionParameter() {collection=&FortniteMaterialParameters, parameterName=PlayerPosition}
  n32 = ComponentMask(None=@n31) {b=true, g=true, r=true}
  n33 = MaterialFunctionCall() {materialFunction=&ObjectPivotPoint}
  Reroute(None=SmoothStep(Min=@n30, Max=@n30, Value=Distance(A=@n32, B=@n33.Object Pivot Location)))
  TextureSample() {texture=/MotherGameplay/Items/AvacadoEater/Assets/Textures/T_Avocado_Eater_Gadget_D}
 ]
}
