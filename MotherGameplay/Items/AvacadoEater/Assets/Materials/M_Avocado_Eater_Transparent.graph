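# Translucent lit material for the Avocado Eater gadget, with textured diffuse, roughness and world-space normal, plus a downward-facing glow color.
# Opacity blends from depth fade near the camera to a default value with distance.
# World position offset runs MF_Avocado_Wobble around an influencer built from random motion, an orientation-rotated offset and sine-driven chomping.
# Custom primitive data exposes Trigger and CollapseVerts, and a Player Direction vector derives from the player position collection parameter.
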Material /MotherGameplay/Items/AvacadoEater/Assets/Materials/M_Avocado_Eater_Transparent
graph {
 properties {
  bTangentSpaceNormal=false
  blendMode=BLEND_Translucent
  pixelDepthOffsetMode=PDOM_Legacy
  translucencyLightingMode=TLM_Surface
  usedShadingModels=Default Lit
 }
 nodes [
  Constant() {r=1}
  LinearInterpolate(
   A=NamedRerouteUsage()
   B=Multiply(A=WorldPosition().XYZ) {constB=-1}
   Alpha=ScalarParameter() {bUseCustomPrimitiveData=true, parameterName=CollapseVerts, primitiveDataIndex=1}
  )
  MaterialFunctionCall!(Alpha, Material Attributes) {materialFunction=/Game/Weapons/WeaponSkins/MaterialFunction/ThinFilmFunction}
  n1 = MaterialFunctionCall(Color=VertexNormalWS()) {materialFunction=&SplitComponents}
  MaterialFunctionCall(X=@n1.R, Y=@n1.R, Z=Saturate(None=@n1.B)) {materialFunction=&MakeFloat3}
  NamedRerouteDeclaration(None=Multiply(
    A=TextureSample() {samplerType=SAMPLERTYPE_Masks, texture=/MotherGameplay/Items/AvacadoEater/Assets/Textures/T_Avocado_Eater_Gadget_S}.B
    B=Constant() {r=0.5}
   )) {name=Roughness, nodeColor=<rgba 0.4062 0 1 1>}
  n2 = DepthFade() {fadeDistanceDefault=25}
  n3 = CameraPositionWS()
  n4 = MaterialFunctionCall() {materialFunction=&ObjectPivotPoint}
  n5 = Divide(A=Subtract(A=Distance(A=@n3, B=@n4.Object Pivot Location)) {constB=1000}) {constB=1000}
  n6 = NamedRerouteDeclaration(None=LinearInterpolate(A=Power(Base=@n2) {constExponent=0.5}, Alpha=Saturate(None=@n5))) {name=Opacity, nodeColor=<rgba 0.5469 0 1 1>}
  NamedRerouteDeclaration(None=Multiply(A=TextureSample() {texture=/MotherGameplay/Items/AvacadoEater/Assets/Textures/T_Avocado_Eater_Gadget_D}.RGB, B=Saturate(None=@n6))) {name=Diffuse, nodeColor=<rgba 0 0.625 1 1>}
  n7 = ComponentMask(None=VertexNormalWS()) {b=true}
  n8 = Power(Base=ConstantBiasScale(None=Multiply(A=@n7) {constB=-1})) {constExponent=5}
  n9 = VectorParameter() {parameterName=Glow Color, defaultValue=<rgba 0.0037 0.212 0.0224 1>}
  NamedRerouteDeclaration(None=Multiply(A=Multiply(A=@n9.RGB, B=@n8), B=NamedRerouteUsage())) {name=Emissive, nodeColor=<rgba 0.8516 1 0 1>}
  NamedRerouteDeclaration(None=TextureSample() {samplerType=SAMPLERTYPE_Masks, texture=/MotherGameplay/Items/AvacadoEater/Assets/Textures/T_Avocado_Eater_Gadget_FX}.B) {name=Shell Mask, nodeColor=<rgba 0 0.8359 1 1>}
  n10 = MaterialFunctionCall() {materialFunction=&ObjectPivotPoint}
  n11 = NamedRerouteUsage()
  n12 = LinearInterpolate(Alpha=@n11) {constA=0.7, constB=0.1}
  n13 = Constant() {r=2}
  n14 = Constant3Vector() {constant=<rgba 0 0 37.5 0>}
  n15 = ObjectOrientation()
  n16 = Multiply(A=LinearInterpolate(Alpha=@n11) {constA=3, constB=0}, B=Time()) {constB=3}
  n17 = Multiply(A=NamedRerouteUsage(), B=ConstantBiasScale(None=Sine(None=@n16)))
  n18 = ScalarParameter() {defaultValue=8, parameterName=Chomp Distance}
  n19 = Add(
   A=Add(
    A=MaterialFunctionCall!(Speed=@n12, Max Distance=@n13) {materialFunction=&/MF_RandomLocation}.Result
    B=MaterialFunctionCall!(VectorToRotate=@n14, LookAtVector=@n15, RestingVector) {materialFunction=/Engine/Functions/Engine_MaterialFunctions01/Vectors/RotateVector}.Result
   )
   B=Multiply(A=LinearInterpolate(Alpha=@n11) {constA=1, constB=0}, B=Multiply(A=@n17, B=@n18) {constB=10})
  )
  NamedRerouteDeclaration(None=MaterialFunctionCall!(Vertex Position, Str, Radius, Gradient, Gradient2, Influencer Position=Add(A=@n10.Object Pivot Location, B=@n19)) {materialFunction=&/MF_Avocado_Wobble}.Result) {name=WPO, nodeColor=<rgba 0.8047 1 0 1>}
  NamedRerouteDeclaration(None=OneMinus(None=ScalarParameter() {bUseCustomPrimitiveData=true, parameterName=Trigger})) {name=Trigger, nodeColor=<rgba 0 0.25 1 1>}
  n20 = CollectionParameter() {collection=&FortniteMaterialParameters, parameterName=PlayerPosition}
  n21 = ComponentMask(None=@n20) {b=true, g=true, r=true}
  n22 = MaterialFunctionCall() {materialFunction=&ObjectPivotPoint}
  NamedRerouteDeclaration(None=MaterialFunctionCall(Vector=Subtract(A=@n21, B=@n22.Object Pivot Location), Default=Constant3Vector() {constant=<rgba 1 0 0 0>}) {materialFunction=&SafeNormalize}.Result) {name=Player Direction, nodeColor=<rgba 1 0.1172 0 1>}
  NamedRerouteUsage()
  NamedRerouteUsage()
  NamedRerouteUsage()
  NamedRerouteUsage()
  n23 = Constant() {r=1000}
  n24 = CollectionParameter() {collection=&FortniteMaterialParameters, parameterName=PlayerPosition}
  n25 = ComponentMask(None=@n24) {b=true, g=true, r=true}
  n26 = MaterialFunctionCall() {materialFunction=&ObjectPivotPoint}
  Reroute(None=SmoothStep(Min=@n23, Max=@n23, Value=Distance(A=@n25, B=@n26.Object Pivot Location)))
  Transform(None=TextureSample() {samplerType=SAMPLERTYPE_Normal, texture=/MotherGameplay/Items/AvacadoEater/Assets/Textures/T_Avocado_Eater_Gadget_N}.RGB)
 ]
}
