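# Surface material for the avocado parasite item, animating it as a chomping creature that turns toward the player through world position offset.
# Player position sets a yaw around the object pivot, with normals corrected for the rotation.
# Time-driven sine and cosine waves combine vertex-color deformation, a chomp travel offset and random jitter.
# Custom primitive data supplies a Trigger value and a CollapseVerts blend that collapses vertices; base color is black.
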Material /MotherGameplay/Items/AvacadoEater/Assets/Materials/M_Avocado_Parasite
graph {
 properties {pixelDepthOffsetMode=PDOM_Legacy, usedShadingModels=Default Lit}
 nodes [
  Add(B=RotateAboutAxis(
    NormalizedRotationAxis=Constant3Vector() {constant=<rgba 0 0 1 0>}
    RotationAngle=NamedRerouteUsage()
    PivotPoint=MaterialFunctionCall() {materialFunction=&ObjectPivotPoint}.Object Pivot Location
    Position=WorldPosition().XYZ
   ) {period=4})
  Constant()
  Constant() {r=1}
  LinearInterpolate(
   A=NamedRerouteUsage()
   B=Multiply(A=WorldPosition().XYZ) {constB=-1}
   Alpha=ScalarParameter() {bUseCustomPrimitiveData=true, parameterName=CollapseVerts, primitiveDataIndex=1}
  )
  n1 = Constant3Vector() {constant=<rgba 0 0 0 1>}
  n2 = Saturate(None=DotProduct(A=SkyAtmosphereLightDirection(), B=VertexNormalWS()))
  LinearInterpolate(
   A=@n1
   B=LinearInterpolate(A=Constant3Vector() {constant=<rgba 0.0232 0.2307 0.0999 1>})
   Alpha=Multiply(A=Power(Base=@n2), B=Fresnel() {baseReflectFraction=0, exponent=2})
  )
  n3 = Constant3Vector() {constant=<rgba 0 0 1 0>}
  n4 = Divide(A=NamedRerouteUsage()) {constB=4}
  NamedRerouteDeclaration(None=Add(B=MaterialFunctionCall(Rotation Axis=@n3, Rotation Angle=@n4) {materialFunction=&FixRotateAboutAxisNormals}.New Tangent Space Vertex Normal)) {name=Normals, nodeColor=<rgba 1 0.8203 0 1>}
  NamedRerouteDeclaration(None=@n1) {name=Base Color, nodeColor=<rgba 0.3594 0 1 1>}
  n5 = Subtract(A=Multiply(A=VertexColor()) {constB=2})
  n6 = NamedRerouteUsage()
  n7 = Multiply(A=Time(), B=LinearInterpolate(Alpha=@n6) {constA=3, constB=0}) {constB=3}
  n8 = ConstantBiasScale(None=Cosine(None=@n7))
  n9 = LinearInterpolate(Alpha=@n6) {constA=1, constB=0}
  n10 = Add(
   A=RotateAboutAxis(
    NormalizedRotationAxis=Constant3Vector() {constant=<rgba 0 0 1 0>}
    RotationAngle=NamedRerouteUsage()
    PivotPoint=MaterialFunctionCall() {materialFunction=&ObjectPivotPoint}.Object Pivot Location
    Position=WorldPosition().XYZ
   ) {period=4}
   B=Multiply(A=Multiply(A=@n5) {constB=-8.6824, desc=Max vertex movement}, B=Multiply(A=@n8, B=@n9))
  )
  n11 = Multiply(
   A=ConstantBiasScale(None=Sine(None=@n7))
   B=Multiply(A=NamedRerouteUsage(), B=ScalarParameter() {defaultValue=10, parameterName=Chomp Travel Distance}) {constB=10}
  )
  n12 = LinearInterpolate(Alpha=NamedRerouteUsage()) {constA=0.7, constB=0.1}
  n13 = Constant() {r=3}
  NamedRerouteDeclaration(None=Add(
    A=Add(A=@n10, B=@n11)
    B=MaterialFunctionCall!(Speed=@n12, Max Distance=@n13) {materialFunction=&/MF_RandomLocation}.Result
   )) {name=WPO, nodeColor=<rgba 0 1 1 1>}
  NamedRerouteDeclaration(None=OneMinus(None=ScalarParameter() {bUseCustomPrimitiveData=true, parameterName=Trigger})) {name=Trigger, nodeColor=<rgba 0 0.2555 1 1>}
  n14 = CollectionParameter() {collection=&FortniteMaterialParameters, parameterName=PlayerPosition}
  n15 = MaterialFunctionCall(
   Vector=Subtract(
    A=ComponentMask(None=@n14) {b=true, g=true, r=true}
    B=MaterialFunctionCall() {materialFunction=&ObjectPivotPoint}.Object Pivot Location
   )
   Default=Constant3Vector() {constant=<rgba 1 0 0 0>}
  ) {materialFunction=&SafeNormalize}
  NamedRerouteDeclaration(None=@n15.Result) {name=Player Direction, nodeColor=<rgba 1 0.5156 0 1>}
  n16 = DotProduct(A=Constant3Vector() {constant=<rgba 0 1 0 0>}, B=@n15.Result)
  n17 = Multiply(A=@n16) {constB=-1}
  n18 = Sign(None=ComponentMask(None=@n15.Result) {r=true})
  NamedRerouteDeclaration(None=LinearInterpolate(A=Add(A=@n17) {constB=2}, B=@n16, Alpha=Saturate(None=@n18))) {name=Rotation Angle, nodeColor=<rgba 1 0.9609 0 1>}
  NamedRerouteUsage()
  NamedRerouteUsage()
  n19 = Constant() {r=1000}
  n20 = CollectionParameter() {collection=&FortniteMaterialParameters, parameterName=PlayerPosition}
  n21 = ComponentMask(None=@n20) {b=true, g=true, r=true}
  n22 = MaterialFunctionCall() {materialFunction=&ObjectPivotPoint}
  Reroute(None=SmoothStep(Min=@n19, Max=@n19, Value=Distance(A=@n21, B=@n22.Object Pivot Location)))
 ]
}
